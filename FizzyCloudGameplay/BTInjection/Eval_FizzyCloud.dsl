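; Behavior tree injection evaluator that decides whether an Athena AI bot can start its FizzyCloud behavior.
; EvaluateStartingConditions first requires the parent check and an enable flag read from a scalable float.
; It then casts to the bot's controller and pawn, passing only when normalized shield is below a scalable-float threshold and the pawn has existed for over three seconds.
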(fn EvaluateStartingConditions
  (params (OwnerComp :type Behavior%20Tree%20Component%20Object%20Reference))
  (bind _returnvalue (|CallParentFunction OwnerComp))
  (if (and _returnvalue (ScalableFloat|GetValueasBool (Variables|Default|GetIsEnabled_Hf)))
    (bind _asfort_athena_aibot_controller (Utilities|Casting|CastToFortAthenaAIBotController (Components|GetOwner OwnerComp))
      (:then
        (bind _asfort_pawn (Utilities|Casting|CastToFortPawn (Pawn|GetControlledPawn _asfort_athena_aibot_controller))
          (:then
            (return (and (< (Math|Float|NormalizetoRange (FortPawn|GetShield _asfort_pawn) 0.0 (FortPawn|GetMaxShield _asfort_pawn)) (ScalableFloat|GetValueatLevel (Variables|Default|GetShieldPercentThreshold_Hf))) (> (Actor|GetGameTimeSinceCreation _asfort_pawn) 3.0))))
          (:CastFailed
            _)))
      (:CastFailed
        _))
    (return false)
    (else
      _)))
